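; AI weapon Blueprint for the HeavyLamp boss's Lobster blade, extending a parent weapon with switchable blade effects.
; SetBladeEffectsActive sets DeactivateBlade, whose OnRep, on cosmetic-capable machines, hides the masked underwater blade mesh or shows the main one and toggles equip audio.
; An override flag blocks the parent's blade visibility event.
; Weapon visibility changes toggle the audio only while the blade is not deactivated.
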(event Custom|SetBladeEffectsActive
  (params (EffectsActive :type Boolean))
  (Variables|Default|SetDeactivateBlade (not EffectsActive)))

(event Custom|OverrideBladeEffects
  (params (OverrideActive :type Boolean))
  (Variables|Default|SetBladeEffectsOverrideActive OverrideActive))

(event EventBeginPlay
  (|CallParentFunction))

(event EventTick
  (params (DeltaSeconds :type Float%20%28single-precision%29)))

(event EventSetBladeVisibility
  (params (Visible :type Boolean))
  (if (not (Variables|Default|GetBladeEffectsOverrideActive))
    (|CallParentFunction Visible)))

(event EventOnWeaponVisibilityChanged
  (params (bVisible :type Boolean) (bSetForLocalControllerOnly :type Boolean))
  (|CallParentFunction bVisible bSetForLocalControllerOnly)
  (if (not (Variables|Default|GetDeactivateBlade))
    (CallFunction|AIWeaponVisibilityChanged bVisible)))

(fn AIWeaponVisibilityChanged
  (params (Visible :type Boolean))
  (bind _weaponequipsoundcomponent_lobster (Variables|Default|GetWeaponEquipSoundComponent_Lobster))
  (if Visible
    (Class|WeaponEquipSoundComponent|PlayAudio _weaponequipsoundcomponent_lobster)
    (else
      (Class|WeaponEquipSoundComponent|StopAudio _weaponequipsoundcomponent_lobster))))

(fn OnRep_DeactivateBlade
  (Networking|CanExecuteCosmeticEvents
    (:True
      (if (Variables|Default|GetDeactivateBlade)
        (Rendering|SetVisibility (Variables|Default|GetSM_Galileo_Lobster_Blade_Masked(VisibleUnderwater)) false true)
        (CallFunction|AIWeaponVisibilityChanged false)
        (else
          (Rendering|SetVisibility (Variables|Default|GetSM_Galileo_Lobster_Blade) true true)
          (CallFunction|AIWeaponVisibilityChanged true))))))

(fn ConstructionScript
  (|CallParentFunction))
